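; Golf ball toy variant (Season 5 Elite) that colors its particle trail to match the owning player.
; On BeginPlay it runs parent logic, waits one tick, then casts the owner to the Athena player state.
; On success it sets the trail's color from the player's pin color and its size from the trail size variable.
; If the cast fails, it applies only a fixed purple trail color.
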(event EventBeginPlay
  (bind _golfballtrail (Variables|Default|GetGolfBallTrail))
  (|CallParentFunction)
  (Utilities|FlowControl|Delay 0.0)
  (bind _asfort_player_state_athena (Utilities|Casting|CastToFortPlayerStateAthena (Actor|GetOwner))
    (:then
      (Effects|Components|ParticleSystem|SetColorParameter _golfballtrail "TrailColor" (Class|FortPlayerStateZone|GetPinColor _asfort_player_state_athena))
      (Effects|Components|ParticleSystem|SetVectorParameter _golfballtrail "TrailSize" (Variables|Default|GetTrailSize)))
    (:CastFailed
      (Effects|Components|ParticleSystem|SetColorParameter _golfballtrail "TrailColor" "(R=0.678014,G=0.000000,B=0.885417,A=1.000000)"))))

(fn ConstructionScript
  (|CallParentFunction))
